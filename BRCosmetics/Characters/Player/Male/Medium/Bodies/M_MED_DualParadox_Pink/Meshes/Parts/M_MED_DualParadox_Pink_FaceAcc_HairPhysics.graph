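# Physics asset for the top hair of the pink DualParadox medium male character's FaceAcc mesh part.
# Nine hair bones each carry an offset capsule body, paired with a small capsule on a matching MSTR bone.
# One constraint per pair uses identical tight swing and twist limits, keeping motion to a narrow sway with almost no twist.
# No constraints link neighboring strands to each other.
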PhysicsAsset /BRCosmetics/Characters/Player/Male/Medium/Bodies/M_MED_DualParadox_Pink/Meshes/Parts/M_MED_DualParadox_Pink_FaceAcc_HairPhysics
bodies {
 dyn_top_hair_01 [{shapeType=Capsule, length=5, radius=5, shapeName=dyn_top_hair_01_capsule, center=<8 0 0>, rotation=<rot 90 0 0>}]
 dyn_top_hair_01_MSTR [{shapeType=Capsule, length=1, radius=1, shapeName=dyn_top_hair_01_MSTR_capsule, rotation=<rot 90 0 0>}]
 dyn_top_hair_02 [{shapeType=Capsule, length=5, radius=5, shapeName=dyn_top_hair_02_capsule, center=<10 0 0>, rotation=<rot 90 0 0>}]
 dyn_top_hair_02_MSTR [{shapeType=Capsule, length=1, radius=1, shapeName=dyn_top_hair_02_MSTR_capsule, rotation=<rot 90 0 0>}]
 dyn_top_hair_03 [{shapeType=Capsule, length=5, radius=5, shapeName=dyn_top_hair_03_capsule, center=<10 0 0>, rotation=<rot 90 0 0>}]
 dyn_top_hair_03_MSTR [{shapeType=Capsule, length=1, radius=1, shapeName=dyn_top_hair_03_MSTR_capsule, rotation=<rot 90 0 0>}]
 dyn_top_hair_04 [{shapeType=Capsule, length=5, radius=5, shapeName=dyn_top_hair_04_capsule, center=<7 0 0>, rotation=<rot 90 0 0>}]
 dyn_top_hair_04_MSTR [{shapeType=Capsule, length=1, radius=1, shapeName=dyn_top_hair_04_MSTR_capsule, rotation=<rot 90 0 0>}]
 dyn_top_hair_05 [{shapeType=Capsule, length=5, radius=5, shapeName=dyn_top_hair_05_capsule, center=<8 0 0>, rotation=<rot 90 0 0>}]
 dyn_top_hair_05_MSTR [{shapeType=Capsule, length=1, radius=1, shapeName=dyn_top_hair_05_MSTR_capsule, rotation=<rot 90 0 0>}]
 dyn_top_hair_06 [{shapeType=Capsule, length=5, radius=5, shapeName=dyn_top_hair_06_capsule, center=<10 0 0>, rotation=<rot 90 0 0>}]
 dyn_top_hair_06_MSTR [{shapeType=Capsule, length=1, radius=1, shapeName=dyn_top_hair_06_MSTR_capsule, rotation=<rot 90 0 0>}]
 dyn_top_hair_07 [{shapeType=Capsule, length=5, radius=5, shapeName=dyn_top_hair_07_capsule, center=<7 0 0>, rotation=<rot 90 0 0>}]
 dyn_top_hair_07_MSTR [{shapeType=Capsule, length=1, radius=1, shapeName=dyn_top_hair_07_MSTR_capsule, rotation=<rot 90 0 0>}]
 dyn_top_hair_08 [{shapeType=Capsule, length=5, radius=5, shapeName=dyn_top_hair_08_capsule, center=<7 0 0>, rotation=<rot 90 0 0>}]
 dyn_top_hair_08_MSTR [{shapeType=Capsule, length=1, radius=1, shapeName=dyn_top_hair_08_MSTR_capsule, rotation=<rot 90 0 0>}]
 dyn_top_hair_09 [{shapeType=Capsule, length=5, radius=5, shapeName=dyn_top_hair_09_capsule, center=<10 0 0>, rotation=<rot 90 0 0>}]
 dyn_top_hair_09_MSTR [{shapeType=Capsule, length=1, radius=1, shapeName=dyn_top_hair_09_MSTR_capsule, rotation=<rot 90 0 0>}]
}
constraints [
 {bone1Name=dyn_top_hair_01, bone2Name=dyn_top_hair_01_MSTR, swing1LimitDegrees=10, swing2LimitDegrees=20, twistLimitDegrees=2}
 {bone1Name=dyn_top_hair_02, bone2Name=dyn_top_hair_02_MSTR, swing1LimitDegrees=10, swing2LimitDegrees=20, twistLimitDegrees=2}
 {bone1Name=dyn_top_hair_03, bone2Name=dyn_top_hair_03_MSTR, swing1LimitDegrees=10, swing2LimitDegrees=20, twistLimitDegrees=2}
 {bone1Name=dyn_top_hair_04, bone2Name=dyn_top_hair_04_MSTR, swing1LimitDegrees=10, swing2LimitDegrees=20, twistLimitDegrees=2}
 {bone1Name=dyn_top_hair_05, bone2Name=dyn_top_hair_05_MSTR, swing1LimitDegrees=10, swing2LimitDegrees=20, twistLimitDegrees=2}
 {bone1Name=dyn_top_hair_06, bone2Name=dyn_top_hair_06_MSTR, swing1LimitDegrees=10, swing2LimitDegrees=20, twistLimitDegrees=2}
 {bone1Name=dyn_top_hair_07, bone2Name=dyn_top_hair_07_MSTR, swing1LimitDegrees=10, swing2LimitDegrees=20, twistLimitDegrees=2}
 {bone1Name=dyn_top_hair_08, bone2Name=dyn_top_hair_08_MSTR, swing1LimitDegrees=10, swing2LimitDegrees=20, twistLimitDegrees=2}
 {bone1Name=dyn_top_hair_09, bone2Name=dyn_top_hair_09_MSTR, swing1LimitDegrees=10, swing2LimitDegrees=20, twistLimitDegrees=2}
]
