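# Anim Blueprint for the DualParadox Pink face accessory, adding physics and animated shake to hair and earrings.
# It copies the parent mesh pose, then runs RigidBody hair and earring simulations, AnimDynamics bangs and additive rotations on hair bones.
# Four noise rotators from game time times shakeFrequency drive those rotations.
# shakeFrequency rises temporarily after AthenaKills increases, or during MaxEnergize emote when HairAmbientMovement notifies set isPowerUp.
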AnimBlueprint /BRCosmetics/Characters/Player/Male/Medium/Bodies/M_MED_DualParadox_Pink/Meshes/Parts/M_MED_DualParadox_Pink_FaceAcc_AnimBP
target_skeleton=&/M_MED_DualParadox_Pink_FaceAcc_Skeleton
graphs [
 AnimGraph {nodes [
   n1 = RigidBody(ComponentPose=Local To Component(LocalPose=Copy Pose From Mesh() {bUseAttachedParent=true})) {
    bClampLinearTranslationLimitToRefPose=true
    bOverrideWorldGravity=true
    overridePhysicsAsset=&/M_MED_DualParadox_Pink_FaceAcc_HairPhysics
    simulationSpace=BaseBoneSpace
    alphaBoolBlend {blendOption=Linear}
    baseBoneRef {boneName=head}
    componentAppliedLinearAccClamp=<1500 1500 1500>
    componentLinearAccScale=<0.1 0.1 0.1>
    componentLinearVelScale=<0.5 0.5 0.5>
    overrideWorldGravity {z=-700}
   }
   n2 = RigidBody(ComponentPose=@n1) {
    bClampLinearTranslationLimitToRefPose=true
    bOverrideWorldGravity=true
    overridePhysicsAsset=&/M_MED_DualParadox_Pink_FaceAcc_EarringPhysics
    simulationSpace=BaseBoneSpace
    alphaBoolBlend {blendOption=Linear}
    baseBoneRef {boneName=spine_05}
    componentAppliedLinearAccClamp=<1000 1000 1000>
    componentLinearAccScale=<0.1 0.1 0.1>
    componentLinearVelScale=<0.4 0.4 0.4>
    overrideWorldGravity {z=-700}
   }
   n3 = "Transform (Modify) Bone - Bone: dyn_top_hair_01"(ComponentPose=@n2, Alpha=0.2, Translation="0,0,0", Rotation=Get HairRotator_01() -> (HairRotator_01:Rotator), Scale="1,1,1") {rotationMode=BMM_Additive, rotationSpace=BCS_BoneSpace, alphaBoolBlend {blendOption=Linear}, boneToModify {boneName=dyn_top_hair_01}}
   n4 = Get HairRotator_03() -> (HairRotator_03:Rotator)
   n5 = "Transform (Modify) Bone - Bone: dyn_top_hair_03"(ComponentPose=@n3, Alpha=0.6, Translation="0,0,0", Rotation=@n4, Scale="1,1,1") {rotationMode=BMM_Additive, rotationSpace=BCS_BoneSpace, alphaBoolBlend {blendOption=Linear}, boneToModify {boneName=dyn_top_hair_03}}
   n6 = "Transform (Modify) Bone - Bone: dyn_top_hair_04"(ComponentPose=@n5, Alpha=0.6, Translation="0,0,0", Rotation=@n4, Scale="1,1,1") {rotationMode=BMM_Additive, rotationSpace=BCS_BoneSpace, alphaBoolBlend {blendOption=Linear}, boneToModify {boneName=dyn_top_hair_04}}
   n7 = "Transform (Modify) Bone - Bone: dyn_top_hair_06"(ComponentPose=@n6, Alpha=0.3, Translation="0,0,0", Rotation=@n4, Scale="1,1,1") {rotationMode=BMM_Additive, rotationSpace=BCS_BoneSpace, alphaBoolBlend {blendOption=Linear}, boneToModify {boneName=dyn_top_hair_06}}
   n8 = "Transform (Modify) Bone - Bone: dyn_top_hair_09"(ComponentPose=@n7, Alpha=0.6, Translation="0,0,0", Rotation=@n4, Scale="1,1,1") {rotationMode=BMM_Additive, rotationSpace=BCS_BoneSpace, alphaBoolBlend {blendOption=Linear}, boneToModify {boneName=dyn_top_hair_09}}
   n9 = Get HairRotator_04() -> (HairRotator_04:Rotator)
   n10 = "Transform (Modify) Bone - Bone: dyn_top_hair_02"(ComponentPose=@n8, Alpha=0.6, Translation="0,0,0", Rotation=@n9, Scale="1,1,1") {rotationMode=BMM_Additive, rotationSpace=BCS_BoneSpace, alphaBoolBlend {blendOption=Linear}, boneToModify {boneName=dyn_top_hair_02}}
   n11 = "Transform (Modify) Bone - Bone: dyn_top_hair_05"(ComponentPose=@n10, Alpha=0.6, Translation="0,0,0", Rotation=@n9, Scale="1,1,1") {rotationMode=BMM_Additive, rotationSpace=BCS_BoneSpace, alphaBoolBlend {blendOption=Linear}, boneToModify {boneName=dyn_top_hair_05}}
   n12 = "Transform (Modify) Bone - Bone: dyn_top_hair_07"(ComponentPose=@n11, Alpha=0.6, Translation="0,0,0", Rotation=@n9, Scale="1,1,1") {rotationMode=BMM_Additive, rotationSpace=BCS_BoneSpace, alphaBoolBlend {blendOption=Linear}, boneToModify {boneName=dyn_top_hair_07}}
   n13 = "Transform (Modify) Bone - Bone: dyn_top_hair_08"(ComponentPose=@n12, Alpha=0.6, Translation="0,0,0", Rotation=@n9, Scale="1,1,1") {rotationMode=BMM_Additive, rotationSpace=BCS_BoneSpace, alphaBoolBlend {blendOption=Linear}, boneToModify {boneName=dyn_top_hair_08}}
   n14 = Anim Dynamics - Bone: dyn_bangs_01_MSTR!(ComponentPose=@n13, GravityScale:float=1, GravityOverride:Vector="0,0,-980", ExternalForce="0,0,0") -> (Pose) {
    alpha=0.65
    angularDampingOverride=0.9999
    angularSpringConstant=200
    bAngularSpring=true
    bGravityOverrideInSimSpace=true
    bOverrideAngularDamping=true
    bUseGravityOverride=true
    simulationSpace=RootRelative
    alphaBoolBlend {blendOption=Linear}
    boundBone {boneName=dyn_bangs_01_MSTR}
    componentAppliedLinearAccClamp=<1000 1000 1000>
    componentLinearAccScale=<0.15 0.15 0.15>
    componentLinearVelScale=<0.5 0.5 0.5>
    gravityOverride {z=-980}
    physicsBodyDefinitions [{
      sphereCollisionRadius=0
      boundBone {boneName=dyn_bangs_01_MSTR}
      boxExtents=<10 16 12>
      constraintSetup {angularLimitsMax=<0 5 10>, angularLimitsMin=<0 -5 -10>, angularTarget=<0 0 0>}
      localJointOffset=<1 0 -0.2>
     }]
    relativeSpaceBone {boneName=head}
   }
   n15 = Anim Dynamics - Bone: dyn_bangs_02_MSTR!(ComponentPose=@n14, GravityScale:float=1, GravityOverride:Vector="0,0,-980", ExternalForce="0,0,0") -> (Pose) {
    alpha=0.65
    angularDampingOverride=1
    angularSpringConstant=200
    bAngularSpring=true
    bGravityOverrideInSimSpace=true
    bOverrideAngularDamping=true
    bUseGravityOverride=true
    simulationSpace=RootRelative
    alphaBoolBlend {blendOption=Linear}
    boundBone {boneName=dyn_bangs_02_MSTR}
    componentAppliedLinearAccClamp=<1000 1000 1000>
    componentLinearAccScale=<0.15 0.15 0.15>
    componentLinearVelScale=<0.5 0.5 0.5>
    gravityOverride {z=-980}
    physicsBodyDefinitions [{
      sphereCollisionRadius=0
      boundBone {boneName=dyn_bangs_02_MSTR}
      boxExtents=<10 16 12>
      constraintSetup {angularLimitsMax=<0 15 10>, angularLimitsMin=<0 -5 -10>, angularTarget=<0 0 0>}
      localJointOffset=<2 -0.3 -2>
     }]
    relativeSpaceBone {boneName=head}
   }
   n16 = Anim Dynamics - Bone: dyn_bangs_03_MSTR!(ComponentPose=@n15, GravityScale:float=1, GravityOverride:Vector="0,0,-980", ExternalForce="0,0,0") -> (Pose) {
    alpha=0.65
    angularDampingOverride=1
    angularSpringConstant=200
    bAngularSpring=true
    bGravityOverrideInSimSpace=true
    bOverrideAngularDamping=true
    bUseGravityOverride=true
    simulationSpace=RootRelative
    alphaBoolBlend {blendOption=Linear}
    boundBone {boneName=dyn_bangs_03_MSTR}
    componentAppliedLinearAccClamp=<1000 1000 1000>
    componentLinearAccScale=<0.15 0.15 0.15>
    componentLinearVelScale=<0.5 0.5 0.5>
    gravityOverride {z=-980}
    physicsBodyDefinitions [{
      sphereCollisionRadius=0
      boundBone {boneName=dyn_bangs_03_MSTR}
      boxExtents=<10 12 16>
      constraintSetup {angularLimitsMax=<0 15 10>, angularLimitsMin=<0 -10 -10>, angularTarget=<0 0 0>}
      localJointOffset=<1.5 -0.8 -1.6>
     }]
    relativeSpaceBone {boneName=head}
   }
   n17 = Anim Dynamics - Bone: dyn_bangs_04_MSTR!(ComponentPose=@n16, GravityScale:float=1, GravityOverride:Vector="0,0,-980", ExternalForce="0,0,0") -> (Pose) {
    alpha=0.65
    angularDampingOverride=0.9999
    angularSpringConstant=200
    bAngularSpring=true
    bGravityOverrideInSimSpace=true
    bOverrideAngularDamping=true
    bUseGravityOverride=true
    simulationSpace=RootRelative
    alphaBoolBlend {blendOption=Linear}
    boundBone {boneName=dyn_bangs_04_MSTR}
    componentAppliedLinearAccClamp=<1000 1000 1000>
    componentLinearAccScale=<0.15 0.15 0.15>
    componentLinearVelScale=<0.5 0.5 0.5>
    gravityOverride {z=-980}
    physicsBodyDefinitions [{
      sphereCollisionRadius=0
      boundBone {boneName=dyn_bangs_04_MSTR}
      boxExtents=<10 16 12>
      constraintSetup {angularLimitsMax=<0 15 20>, angularLimitsMin=<0 -15 -20>, angularTarget=<0 0 0>}
      localJointOffset=<1 0 0>
     }]
    relativeSpaceBone {boneName=head}
   }
   n18 = Anim Dynamics - Bone: dyn_bangs_05_MSTR!(ComponentPose=@n17, GravityScale:float=1, GravityOverride:Vector="0,0,-980", ExternalForce="0,0,0") -> (Pose) {
    alpha=0.65
    angularDampingOverride=0.9999
    angularSpringConstant=200
    bAngularSpring=true
    bGravityOverrideInSimSpace=true
    bOverrideAngularDamping=true
    bUseGravityOverride=true
    simulationSpace=RootRelative
    alphaBoolBlend {blendOption=Linear}
    boundBone {boneName=dyn_bangs_05_MSTR}
    componentAppliedLinearAccClamp=<1000 1000 1000>
    componentLinearAccScale=<0.15 0.15 0.15>
    componentLinearVelScale=<0.5 0.5 0.5>
    gravityOverride {z=-980}
    physicsBodyDefinitions [{
      sphereCollisionRadius=0
      boundBone {boneName=dyn_bangs_05_MSTR}
      boxExtents=<10 16 12>
      constraintSetup {angularLimitsMax=<0 20 30>, angularLimitsMin=<0 -20 -30>, angularTarget=<0 0 0>}
      localJointOffset=<1 0 0>
     }]
    relativeSpaceBone {boneName=head}
   }
   n19 = Get HairRotator_01()
   n20 = "Transform (Modify) Bone - Bone: dyn_bangs_01"(ComponentPose=@n18, Alpha=0.1, Translation="0,0,0", Rotation=@n19, Scale="1,1,1") {rotationMode=BMM_Additive, rotationSpace=BCS_BoneSpace, alphaBoolBlend {blendOption=Linear}, boneToModify {boneName=dyn_bangs_01}}
   n21 = "Transform (Modify) Bone - Bone: dyn_bangs_02"(ComponentPose=@n20, Alpha=0.2, Translation="0,0,0", Rotation=@n19, Scale="1,1,1") {rotationMode=BMM_Additive, rotationSpace=BCS_BoneSpace, alphaBoolBlend {blendOption=Linear}, boneToModify {boneName=dyn_bangs_02}}
   n22 = "Transform (Modify) Bone - Bone: dyn_bangs_03"(ComponentPose=@n21, Alpha=0.2, Translation="0,0,0", Rotation=@n19, Scale="1,1,1") {rotationMode=BMM_Additive, rotationSpace=BCS_BoneSpace, alphaBoolBlend {blendOption=Linear}, boneToModify {boneName=dyn_bangs_03}}
   n23 = Get HairRotator_02() -> (HairRotator_02:Rotator)
   n24 = "Transform (Modify) Bone - Bone: dyn_bangs_04"(ComponentPose=@n22, Alpha=0.2, Translation="0,0,0", Rotation=@n23, Scale="1,1,1") {rotationMode=BMM_Additive, rotationSpace=BCS_BoneSpace, alphaBoolBlend {blendOption=Linear}, boneToModify {boneName=dyn_bangs_04}}
   n25 = "Transform (Modify) Bone - Bone: dyn_bangs_05"(ComponentPose=@n24, Alpha=0.2, Translation="0,0,0", Rotation=@n23, Scale="1,1,1") {rotationMode=BMM_Additive, rotationSpace=BCS_BoneSpace, alphaBoolBlend {blendOption=Linear}, boneToModify {boneName=dyn_bangs_05}}
   Output Pose(Result=Component To Local(ComponentPose=@n25)) {layerGroup=None}
  ]}
 EventGraph {nodes [
   Event BlueprintUpdateAnimation()
   > n1 = Sequence!(execute) -> (then_0, then_1, then_2:Exec, then_3:Exec)
   > n2 = WatchCosmeticStat!(
    execute:Exec
    Pawn:Fort Player Pawn Object Reference=Get Pawn() -> (Pawn:Fort Player Pawn Object Reference)
    StatTag:Fort Stat Manager Tag Structure=(TagName="StatManager.AthenaKills")
    DefaultValue:int=0
    bBroadcastInitialValue:bool=true
    InitialValueBroadcastDelay:float=-1
    ChangeValueBroadcastDelay:float=-1
    bConsumeWaitingNotifications:bool=true
   ) -> (then:Exec, OnStatChanged:Exec, StatName:Name, StatValue:int) {is=title K2Node_AsyncAction}
   n3 = Set killCount(execute=@n2.OnStatChanged, killCount:double="To Float (Integer)"(InInt=@n2.StatValue)) -> (Output_Get:double)
   > Branch(Condition="float > float"(A=@n3, B=Get Last_KillCount() -> (Last_KillCount:double)) {is=title K2Node_CallFunction})
   > Set ActiveKill(ActiveKill:bool=true) -> (Output_Get:bool)
   > Set Last_KillCount(Last_KillCount:double=@n3) -> (Output_Get:double)
   n4 = GetAnimInstance(self=GetOwnerSkeletalMeshComponent())
   Is Valid(exec=@n1.then_2, InputObject=@n4)
   > n5 = Branch(Condition=Montage_IsPlaying(self=@n4, Montage=/BRCosmetics/Animation/Game/MainPlayer/Emotes/MaxEnergize/Emote_MaxEnergize_CMM_M))
   > n6 = Branch(Condition=Get isPowerUp() -> (isPowerUp:bool))
   > Set shakeFrequency(shakeFrequency:double=FInterpTo(Current=Get shakeFrequency() -> (shakeFrequency:double), Target=2, DeltaTime=GetGameTimeInSeconds(), InterpSpeed=30)) -> (Output_Get:double)
   Set shakeFrequency(execute=@n6.else, shakeFrequency=FInterpTo(Current=Get shakeFrequency(), Target=0.4, DeltaTime=GetGameTimeInSeconds(), InterpSpeed=30))
   n7 = Branch(execute=@n5.else, Condition=Get ActiveKill() -> (ActiveKill:bool))
   n8 = GetGameTimeInSeconds()
   n9 = Get shakeFrequency()
   > Set shakeFrequency(shakeFrequency=FInterpTo(Current=@n9, Target=1.5, DeltaTime=@n8, InterpSpeed=10))
   > Set Timer(Timer:double=float + float(A=Get Timer() -> (Timer:double), B=1)) -> (Output_Get:double)
   n10 = float - float(A=60, B=Get Timer())
   > Branch(Condition="float <= float"(A=MapRangeClamped(Value=@n10, InRangeA=80, InRangeB=0, OutRangeA=1, OutRangeB=0)) -> (ReturnValue=~))
   > Set ActiveKill(ActiveKill=false)
   > Set Timer(Timer=0)
   Set shakeFrequency(execute=@n7.else, shakeFrequency=FInterpTo(Current=@n9, Target=0.4, DeltaTime=@n8, InterpSpeed=50))
   n11 = float * float(A=GetGameTimeInSeconds(), B=Get shakeFrequency()) {is=title K2Node_PromotableOperator}
   n12 = RandomRotatorFunction!(
    execute=@n1.then_3
    GameTimeInSeconds:double=@n11
    PerlinScale:double=1
    Frequency:Rotator="12,20,0"
    Amplitude:Rotator="2,6,0"
    Offset:Rotator="0, 0, 0"
    TimeOffset:double=0
   ) -> (then, RandomRotator:Rotator)
   > Set HairRotator_01(HairRotator_01:Rotator=@n12) -> (Output_Get:Rotator)
   > n13 = RandomRotatorFunction(GameTimeInSeconds=@n11, PerlinScale=1, Frequency="10,18,0", Amplitude="2,5,0", Offset="0,3, 0", TimeOffset=2)
   > Set HairRotator_02(HairRotator_02:Rotator=@n13) -> (Output_Get:Rotator)
   > n14 = RandomRotatorFunction(GameTimeInSeconds=@n11, PerlinScale=1, Frequency="20,5,0", Amplitude="5,3,0", Offset="0, 0, 0", TimeOffset=0)
   > Set HairRotator_03(HairRotator_03:Rotator=@n14) -> (Output_Get:Rotator)
   > n15 = RandomRotatorFunction(GameTimeInSeconds=@n11, PerlinScale=1, Frequency="18,5,0", Amplitude="6,3,0", Offset="0, 0, 0", TimeOffset=0)
   > Set HairRotator_04(HairRotator_04:Rotator=@n15) -> (Output_Get:Rotator)
   Event BlueprintInitializeAnimation()
   > Set Skel_Mesh(Skel_Mesh=GetOwnerSkeletalMeshComponent())
   > n16 = Cast To FortPlayerPawn(Object=TryGetFortPawn())
   > Set Pawn(Pawn:Fort Player Pawn Object Reference=@n16) -> (Output_Get:Fort Player Pawn Object Reference)
   n17 = Event OnNotifyPartsStopped!() -> (OutputDelegate, then, Context:Name=None) {is=title K2Node_Event}
   > Branch(Condition="Equal (Name)"(A=@n17.Context, B=HairAmbientMovement))
   > Set isPowerUp(isPowerUp:bool=false) -> (Output_Get:bool)
   n18 = Event OnNotifyPartsStarted!() -> (OutputDelegate, then, Context:Name=None) {is=title K2Node_Event}
   > Branch(Condition="Equal (Name)"(A=@n18.Context, B=HairAmbientMovement))
   > Set isPowerUp(isPowerUp=true)
  ]}
]
variables [
 ActiveKill {type=boolean}
 HairRotator_01 {type=Rotator}
 HairRotator_02 {type=Rotator}
 HairRotator_03 {type=Rotator}
 HairRotator_04 {type=Rotator}
 isPowerUp {type=boolean}
 killCount {type=number}
 Last_KillCount {type=number}
 Pawn {type=FortPlayerPawn}
 shakeFrequency {type=number, default=0.3}
 Skel_Mesh {type=SkeletalMeshComponent}
 Timer {type=number}
]
